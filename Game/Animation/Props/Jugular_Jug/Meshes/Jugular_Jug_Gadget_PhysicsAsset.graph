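# Physics asset for the Jugular Jug gadget prop's skeletal mesh.
# A single capsule body on the main bone approximates the jug's volume, offset and rotated to fit the mesh.
# It defines no constraints, so the prop simulates or collides as one rigid piece.
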PhysicsAsset /Game/Animation/Props/Jugular_Jug/Meshes/Jugular_Jug_Gadget_PhysicsAsset
bodies {main [{shapeType=Capsule, length=15.4787, radius=11.13, center=<-0.5038 0.0019 3.8732>, rotation=<rot -0.0269 -90 -10.1048>}]}
constraints []
